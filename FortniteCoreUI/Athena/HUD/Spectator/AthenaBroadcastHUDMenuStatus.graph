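# Spectator broadcast HUD widget that shows an icon for which menu is active, map or inventory, built on AthenaBroadcastHUDMenuStatusBase.
# A borderless CommonBorder wraps a single image whose brush texture is swapped by the parent class's InventoryActivated and MapActivated events.
# PreConstruct sets the map icon as the initial state, replacing the backpack texture saved on the image.
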WidgetBlueprint /FortniteCoreUI/Athena/HUD/Spectator/AthenaBroadcastHUDMenuStatus
graphs [EventGraph {nodes [
   Event PreConstruct()
   > SetBrushFromTexture!(
    execute
    self:Image Object Reference=Get Image_0() -> (Image_0:Image Object Reference)
    Texture:Texture 2D Object Reference=Get MapIcon() -> (MapIcon:Texture 2D Object Reference)
    bMatchSize:bool=false
   ) -> (then)
   Event InventoryActivated() {is=title K2Node_Event}
   > SetBrushFromTexture(self=Get Image_0(), Texture=Get InventoryIcon() -> (InventoryIcon:Texture 2D Object Reference), bMatchSize=false)
   Event MapActivated() {is=title K2Node_Event}
   > SetBrushFromTexture(self=Get Image_0(), Texture=Get MapIcon(), bMatchSize=false)
  ]}]
variables [
 InventoryIcon {type=Texture2D, default=/Game/Athena/HUD/Art/T-Icon-Callout-Backpack}
 MapIcon {type=Texture2D, default=/Game/Athena/HUD/Art/T-Icon-Callout-Map}
]
widgets {
 info {
  inheritedWidgetCount=1
  parentClass=/Script/FortniteUI.AthenaBroadcastHUDMenuStatusBase
  rootWidgetClass=&CommonBorder
  widgetCount=3
 }
 widgets [
  Background {
   widgetClassPath=&CommonBorder
   bIsVariable=true
   widget=Background
   properties {
    background {
     bIsDynamicallyLoaded=false
     drawAs=NoDrawType
     imageType=NoImage
     mirroring=NoMirror
     resourceName=None
     resourceObject=null
     tiling=NoTile
     imageSize=<32 32>
     margin=<ltrb 1 1 1 1>
     outlineSettings {
      bUseBrushTransparency=false
      roundingType=HalfHeightRadius
      width=0
      color {colorUseRule=UseColor_Specified, specifiedColor=<rgba 0 0 0 0>}
      cornerRadii=<0 0 0 0>
     }
     tintColor {colorUseRule=UseColor_Specified, specifiedColor=<rgba 1 1 1 1>}
     uVRegion {bIsValid=false, max=<0 0>, min=<0 0>}
    }
    padding=<ltrb 0 0 0 0>
   }
  }
  Image_0 {
   widgetClassPath=&Image
   bIsVariable=true
   parent=Background
   widget=Image_0
   properties {
    brush {
     bIsDynamicallyLoaded=false
     drawAs=Image
     imageType=NoImage
     mirroring=NoMirror
     resourceName=None
     resourceObject=/Game/Athena/HUD/Art/T-Icon-Backpack-64-S
     tiling=NoTile
     imageSize=<128 128>
     margin=<ltrb 0 0 0 0>
     outlineSettings {
      bUseBrushTransparency=false
      roundingType=HalfHeightRadius
      width=0
      color {colorUseRule=UseColor_Specified, specifiedColor=<rgba 0 0 0 0>}
      cornerRadii=<0 0 0 0>
     }
     tintColor {colorUseRule=UseColor_Specified, specifiedColor=<rgba 1 1 1 1>}
     uVRegion {bIsValid=false, max=<0 0>, min=<0 0>}
    }
    renderTransform {angle=0, scale=<1 1>, shear=<0 0>, translation=<125 0>}
   }
   slot {class=BorderSlot, properties {padding=<ltrb 0 0 0 0>}}
  }
  TutorialHighlight {widgetClassPath=&TutorialHighlight}
 ]
}
