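# Translucent UI material that draws an animated, team-colored glowing outline, named for the first-place nameplate in spectating.
# A rounded-box signed distance field from remapped UVs, sized by width, height and corner parameters, is rippled by two time-scrolled cosines of the polar angle.
# Smoothstep bands on that distance form three progressively thinner rings stepping outward, the outer two slightly hue-shifted from TeamColor.
# The tinted rings sum into color, and their masks sum into what appears to be opacity.
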Material /Game/Spectating/Effects/FirstPlaceNameplateOutline
graph {
 properties {blendMode=BLEND_Translucent, materialDomain=MD_UI, pixelDepthOffsetMode=PDOM_Legacy, usedShadingModels=Default Lit}
 nodes [
  n1 = Constant3Vector()
  n2 = VectorParameter() {parameterName=TeamColor, defaultValue=<rgba 1 0.25 0 1>}
  n3 = Subtract(A=Multiply(A=TextureCoordinate(), B=Constant2Vector() {g=-2, r=2}) {constB=2}, B=Constant2Vector() {g=-1, r=1}) {constB=-1}
  n4 = Arctangent2Fast(Y=ComponentMask(None=@n3) {r=true}, X=ComponentMask(None=@n3) {g=true})
  n5 = Time()
  n6 = Cosine(None=Add(A=Multiply(A=@n4), B=Multiply(A=@n5) {constB=0.5}))
  n7 = Cosine(None=Add(A=Multiply(A=@n5) {constB=-0.75}, B=Multiply(A=@n4) {constB=5}))
  n8 = ScalarParameter() {defaultValue=0.86, parameterName=BoxWidth, sliderMax=1}
  n9 = ScalarParameter() {defaultValue=0.56, parameterName=BoxHeight, sliderMax=1}
  n10 = Add(
   A=Subtract(
    A=Abs(None=@n3)
    B=MaterialFunctionCall(X=@n8, Y=@n9) {materialFunction=&MakeFloat2}.Result
   )
   B=ScalarParameter() {defaultValue=0.05, parameterName=CornerRadius}
  )
  n11 = MaterialFunctionCall(Vector 2=Max(A=@n10, B=Constant())) {materialFunction=/Engine/Functions/Engine_MaterialFunctions02/Utility/VectorLength}
  n12 = Min(A=Max(A=ComponentMask(None=@n10) {r=true}, B=ComponentMask(None=@n10) {g=true}), B=Constant())
  n13 = Add(
   A=Multiply(B=Multiply(A=@n6, B=@n7)) {constA=0.02, constB=0.02}
   B=Subtract(A=Add(A=@n11.V2 Length, B=@n12), B=ScalarParameter() {parameterName=CornerRadius}) {constB=0.05}
  )
  n14 = OneMinus(None=SmoothStep(Value=Abs(None=@n13)) {constMax=0.06})
  n15 = Constant3Vector()
  n16 = Add(A=@n13) {constB=-0.06}
  n17 = OneMinus(None=SmoothStep(Value=Abs(None=@n16)) {constMax=0.04})
  n18 = MaterialFunctionCall(Hue Shift Percentage=Constant() {r=-0.03}, Texture=VectorParameter() {parameterName=TeamColor, defaultValue=<rgba 1 0.25 0 1>}.RGB) {materialFunction=&HueShift}
  n19 = Add(A=@n13) {constB=-0.1}
  n20 = OneMinus(None=SmoothStep(Value=Abs(None=@n19)) {constMax=0.02})
  n21 = Constant() {r=-0.05}
  n22 = VectorParameter() {parameterName=TeamColor, defaultValue=<rgba 1 0.25 0 1>}
  Add(
   A=Add(A=LinearInterpolate(A=@n1, B=@n2.RGB, Alpha=@n14), B=LinearInterpolate(A=@n15, B=@n18.Result, Alpha=@n17))
   B=LinearInterpolate(
    A=Constant3Vector()
    B=MaterialFunctionCall(Hue Shift Percentage=@n21, Texture=@n22.RGB) {materialFunction=&HueShift}.Result
    Alpha=@n20
   )
  )
  Add(A=Add(A=@n14, B=@n17), B=@n20)
 ]
}
