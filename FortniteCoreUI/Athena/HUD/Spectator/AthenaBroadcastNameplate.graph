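# Broadcast spectator nameplate widget showing a player's name, team number and color, health, shield and vehicle bars, and session rank.
# Parent-class events drive it: detail state switches between full plate, name-only, chevron arrow or hidden, and distance state hides the background beyond max range.
# A dynamic outline material takes the team color and is revealed around first place when session ranks are enabled.
# PreConstruct previews detail and vehicle states at design time.
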WidgetBlueprint /FortniteCoreUI/Athena/HUD/Spectator/AthenaBroadcastNameplate
graphs [
 EventGraph {nodes [
   n1 = Event OnDetailStateChanged!() -> (OutputDelegate, then, NewDetailState:EAthenaSpectatorNameplateDetailState Enum=High) {is=title K2Node_Event}
   > n2 = Switch on EAthenaSpectatorNameplateDetailState!(execute, Selection:EAthenaSpectatorNameplateDetailState Enum=@n1.NewDetailState) -> (High:Exec, Low:Exec, Arrow:Exec, Off:Exec) {is=title K2Node_SwitchEnum}
   n3 = Get Background() -> (Background:Common Border Object Reference)
   SetVisibility(execute=@n2.Arrow, self=@n3, InVisibility=Collapsed)
   > SetVisibility(self=Get Chevron() -> (Chevron:Image Object Reference), InVisibility=SelfHitTestInvisible)
   > SetVisibility(self=Get SizeBox() -> (SizeBox:Size Box Object Reference), InVisibility=SelfHitTestInvisible)
   SetVisibility(execute=@n2.High, self=Get Background(), InVisibility=SelfHitTestInvisible)
   > SetVisibility(self=Get Chevron(), InVisibility=Collapsed)
   > SetVisibility(
    self=[Get HealthBarSB() -> (HealthBarSB:Size Box Object Reference), Get ShieldBarSB() -> (ShieldBarSB:Size Box Object Reference)]
    InVisibility=SelfHitTestInvisible
   )
   > SetVisibility(
    self=Get VehicleHealthSB() -> (VehicleHealthSB:Size Box Object Reference)
    InVisibility=Select(Option 0=Collapsed, Option 1=SelfHitTestInvisible, Index=Get InVehicle() -> (InVehicle:bool))
   )
   > SetHeightOverride(
    self=[Get HealthBarSB(), Get ShieldBarSB(), Get VehicleHealthSB()]
    InHeightOverride=Get HitPointBarFarHeight() -> (HitPointBarFarHeight:double)
   )
   > SetVisibility(self=Get SizeBox(), InVisibility=SelfHitTestInvisible)
   SetVisibility(execute=@n2.Low, self=@n3, InVisibility=SelfHitTestInvisible)
   > SetVisibility(self=[Get HealthBarSB(), Get ShieldBarSB(), Get VehicleHealthSB()], InVisibility=Collapsed)
   > SetVisibility(self=Get Chevron(), InVisibility=Collapsed)
   > SetVisibility(self=Get SizeBox(), InVisibility=SelfHitTestInvisible)
   SetVisibility(execute=@n2.Off, self=Get SizeBox(), InVisibility=Collapsed)
   n4 = Event OnSquadIdChanged!() -> (OutputDelegate, then, InPlayerData:Athena Remote Player View Data Object Reference, InSquadId:Byte=0) {is=title K2Node_Event}
   > "SetText (Text)"(self=Get TeamNumber() -> (TeamNumber:Common Text Object Reference), InText="To Text (Byte)"!(Value:Byte=@n4.InSquadId) -> (ReturnValue:Text))
   n5 = Event OnPlayerDataSet!() -> (OutputDelegate, then, InPlayerData:Athena Remote Player View Data Object Reference) {is=title K2Node_Event}
   > SetPlayerData!(
    execute
    self:Athena Spectator Hit Point Bar Base Object Reference=[
     Get VehicleHealthBar() -> (VehicleHealthBar:Athena Spectator Hit Point Bar Object Reference)
     Get ShieldBar() -> (ShieldBar:Athena Spectator Hit Point Bar Object Reference)
     Get HealthBar() -> (HealthBar:Athena Spectator Hit Point Bar Object Reference)
    ]
    InPlayerData:Athena Remote Player View Data Object Reference=@n5.InPlayerData
   ) -> (then)
   n6 = Event OnPlayerNameChanged!() -> (OutputDelegate, then, InPlayerData:Athena Remote Player View Data Object Reference, InPlayerName:String) {is=title K2Node_Event}
   > "SetText (Text)"(self=Get PlayerName() -> (PlayerName:Common Text Object Reference), InText="To Text (String)"(InString=@n6.InPlayerName))
   n7 = Event OnVehicleStateChanged!() -> (OutputDelegate, then, NewVehicle:Actor Object Reference, OldVehicle:Actor Object Reference) {is=title K2Node_Event}
   n8 = IsValid(Object=@n7.NewVehicle)
   > SetVisibility(self=Get VehicleHealthSB(), InVisibility=Select(Option 0=Collapsed, Option 1=SelfHitTestInvisible, Index=@n8))
   > Set InVehicle(InVehicle:bool=@n8) -> (Output_Get:bool)
   Event Construct()
   > n9 = CreateDynamicMaterialInstance!(execute, Parent=/Game/Spectating/Effects/FirstPlaceNameplateOutline, OptionalName=None, CreationFlags=None) -> (then, ReturnValue)
   > Set FirstPlace Material(FirstPlace Material:Material Instance Dynamic Object Reference=@n9) -> (Output_Get:Material Instance Dynamic Object Reference)
   > SetBrushFromMaterial(
    self:Border Object Reference=Get Background()
    Material=Get FirstPlace Material() -> (FirstPlace Material:Material Instance Dynamic Object Reference)
   )
   n10 = Event OnDistanceStateChanged!() -> (OutputDelegate, then, NewDistanceState:EAthenaSpectatorNameplateDistanceState Enum=Near) {is=title K2Node_Event}
   > SetRenderOpacity(
    self=Get Background()
    InOpacity=SelectFloat(
     A=0
     B=1
     bPickA="Equal (Enum)"(A:EAthenaSpectatorNameplateDistanceState Enum=@n10.NewDistanceState, B:EAthenaSpectatorNameplateDistanceState Enum=FurtherThanMaxDistance)
    )
   )
   n11 = Event OnSessionRankChanged!() -> (OutputDelegate, then, InPlayerData:Athena Remote Player View Data Object Reference, InGamerank:int=0) {is=title K2Node_Event}
   > Set InGameRank(InGameRank:int=@n11.InGamerank) -> (Output_Get:int=0)
   n12 = "integer >= integer"!(A:int=Get InGameRank() -> (InGameRank:int=0), B:int=0) -> (ReturnValue:bool) {is=title K2Node_CallFunction}
   > n13 = Branch(Condition=@n12)
   > n14 = "SetText (Text)"(self=Get SessionRank() -> (SessionRank:Common Text Object Reference), InText="To Text (Integer)"(Value=Get InGameRank() -> (InGameRank=0)))
   Toggle Session Rank Visible!(execute=[@n14.then, @n26.then], self, Visible:bool=Get SessionRanksEnabled() -> (SessionRanksEnabled:bool)) -> (then)
   > Update First Place Background Colors()
   Toggle Session Rank Visible(execute=[@n26.else, @n13.else], Visible=false)
   > Update First Place Background Colors()
   n15 = Event SetSquadIdVisibility!() -> (OutputDelegate, then, bSquadIdVisible:bool) {is=title K2Node_Event}
   > n16 = Branch(Condition=@n15.bSquadIdVisible)
   n17 = Get TeamSB() -> (TeamSB:Size Box Object Reference)
   > SetVisibility(self=@n17, InVisibility=SelfHitTestInvisible)
   > Set bOverrideNameplateScaleToMax(bOverrideNameplateScaleToMax:bool=true) -> (Output_Get:bool)
   SetVisibility(execute=@n16.else, self=@n17, InVisibility=Collapsed)
   > Set bOverrideNameplateScaleToMax(bOverrideNameplateScaleToMax=false)
   n18 = Event SetTeamColorVisible!() -> (OutputDelegate, then, bTeamColorVisible:bool) {is=title K2Node_Event}
   > SetVisibility(
    self=Get TeamColor2SB() -> (TeamColor2SB:Size Box Object Reference)
    InVisibility=Select(Option 0=Collapsed, Option 1=SelfHitTestInvisible, Index=@n18.bTeamColorVisible)
   )
   n19 = Event PreConstruct()
   > Branch(Condition=@n19.IsDesignTime)
   > Set InVehicle(InVehicle=Get DesignTime_InVehicle() -> (DesignTime_InVehicle:bool))
   > OnDetailStateChanged!(
    execute
    self:Athena Spectator Nameplate Base Object Reference
    NewDetailState:EAthenaSpectatorNameplateDetailState Enum=Get DesignTime_DetailState() -> (DesignTime_DetailState:EAthenaSpectatorNameplateDetailState Enum=High)
   ) -> (then)
   n20 = Event OnTeamColorChanged!() -> (OutputDelegate, then, InPlayerData:Athena Remote Player View Data Object Reference, InTeamColor:Linear Color Structure) {is=title K2Node_Event}
   > SetBrushColor(
    self=[Get TeamColor2() -> (TeamColor2:Common Border Object Reference), Get TeamColor() -> (TeamColor:Common Border Object Reference)]
    InBrushColor=@n20.InTeamColor
   )
   > SetVectorParameterValue(self=Get FirstPlace Material(), ParameterName=TeamColor, Value=@n20.InTeamColor)
   > Branch(Condition=Get ShouldColorArrow() -> (ShouldColorArrow:bool))
   > SetColorAndOpacity(self:Image Object Reference=Get Chevron(), InColorAndOpacity:Linear Color Structure=@n20.InTeamColor)
   n21 = Event SetArrowTeamColorVisible!() -> (OutputDelegate, then, bArrowTeamColorVisible:bool) {is=title K2Node_Event}
   > n22 = Set ShouldColorArrow(ShouldColorArrow:bool=@n21.bArrowTeamColorVisible) -> (Output_Get:bool)
   > n23 = Branch(Condition=@n22)
   n24 = Get Chevron()
   > SetColorAndOpacity(self=@n24, InColorAndOpacity=Get BrushColor!(self:Border Object Reference=Get TeamColor()) -> (BrushColor:Linear Color Structure))
   SetColorAndOpacity(execute=@n23.else, self=@n24, InColorAndOpacity=(R=1,G=1,B=1,A=1))
   n25 = Event HandleDisplaySessionRankChanged!() -> (OutputDelegate, then, bDisplay:bool) {is=title K2Node_Event}
   > Set SessionRanksEnabled(SessionRanksEnabled:bool=@n25.bDisplay) -> (Output_Get:bool)
   > n26 = Branch(Condition=@n12)
  ]}
 Toggle Session Rank Visible {nodes [
   n1 = Toggle Session Rank Visible!() -> (then, Visible:bool=~) {is=title K2Node_FunctionEntry}
   > SetVisibility(
    self=Get SessionRankBorder() -> (SessionRankBorder:Common Border Object Reference)
    InVisibility=Select(Option 0=Collapsed, Option 1=Visible, Index=@n1)
   )
  ]}
 Update First Place Background Colors {nodes [
   Update First Place Background Colors() {is=title K2Node_FunctionEntry}
   n1 = AND Boolean(A=Get SessionRanksEnabled(), B="Equal (Integer)"(A=Get InGameRank() -> (InGameRank=0), B=1))
   n2 = Get Background()
   > SetPadding!(
    execute
    self:Border Object Reference=@n2
    InPadding:Margin Structure=Select(
     Option 0:Margin Structure=Make Margin(Left=0, Top=0, Right=0, Bottom=0)
     Option 1:Margin Structure=Make Margin(Left=25, Top=25, Right=25, Bottom=25)
     Index=@n1
    ) -> (ReturnValue:Margin Structure)
   ) -> (then)
   > SetBrushColor(
    self=@n2
    InBrushColor=Select(
     Option 0:Linear Color Structure=Make Linear Color!(R:float=0, G:float=0, B:float=0, A:float=0) -> (LinearColor:Linear Color Structure) {is=title K2Node_MakeStruct}
     Option 1:Linear Color Structure=Make Linear Color(R=1, G=1, B=1, A=1)
     Index=@n1
    ) -> (ReturnValue:Linear Color Structure)
   )
  ]}
 UpdateBackgroundRenderScale {nodes [UpdateBackgroundRenderScale() {is=title K2Node_FunctionEntry}]}
]
variables [
 DesignTime_DetailState {type=EAthenaSpectatorNameplateDetailState, default=High}
 DesignTime_InVehicle {type=boolean}
 FirstPlace Material {type=MaterialInstanceDynamic}
 HitPointBarFarHeight {type=number, default=6}
 InGameRank {type=integer, default=-1}
 InVehicle {type=boolean}
 SessionRanksEnabled {type=boolean}
 ShouldColorArrow {type=boolean}
]
widgets {
 info {inheritedWidgetCount=1, parentClass=/Script/FortniteUI.AthenaSpectatorNameplateBase, rootWidgetClass=&SizeBox, widgetCount=27}
 widgets [
  SizeBox {widgetClassPath=&SizeBox, bIsVariable=true, widget=SizeBox, properties {widthOverride=195}}
  Overlay_1 {widgetClassPath=&Overlay, parent=SizeBox, widget=Overlay_1, slot {class=SizeBoxSlot}}
  Chevron {
   widgetClassPath=&Image
   bIsVariable=true
   parent=Overlay_1
   widget=Chevron
   properties {
    visibility=Collapsed
    brush {
     bIsDynamicallyLoaded=false
     drawAs=Image
     imageType=NoImage
     mirroring=NoMirror
     resourceName=None
     resourceObject=/Game/Athena/HUD/Art/T-Icon-PlayerCarrot
     tiling=NoTile
     imageSize=<70 70>
     margin=<ltrb 0 0 0 0>
     outlineSettings {
      bUseBrushTransparency=false
      roundingType=HalfHeightRadius
      width=0
      color {colorUseRule=UseColor_Specified, specifiedColor=<rgba 0 0 0 0>}
      cornerRadii=<0 0 0 0>
     }
     tintColor {colorUseRule=UseColor_Specified, specifiedColor=<rgba 1 1 1 0.8>}
     uVRegion {bIsValid=false, max=<0 0>, min=<0 0>}
    }
   }
   slot {class=OverlaySlot, properties {horizontalAlignment=HAlign_Center, verticalAlignment=VAlign_Center}}
  }
  Background {
   widgetClassPath=&CommonBorder
   bIsVariable=true
   parent=Overlay_1
   widget=Background
   properties {
    background {
     bIsDynamicallyLoaded=false
     drawAs=Image
     imageType=NoImage
     mirroring=NoMirror
     resourceName=None
     resourceObject=null
     tiling=NoTile
     imageSize=<32 32>
     margin=<ltrb 1 1 1 1>
     outlineSettings {
      bUseBrushTransparency=false
      roundingType=HalfHeightRadius
      width=0
      color {colorUseRule=UseColor_Specified, specifiedColor=<rgba 0 0 0 0>}
      cornerRadii=<0 0 0 0>
     }
     tintColor {colorUseRule=UseColor_Specified, specifiedColor=<rgba 1 1 1 1>}
     uVRegion {bIsValid=false, max=<0 0>, min=<0 0>}
    }
    brushColor=<rgba 1 1 1 0>
    padding=<ltrb 0 0 0 0>
   }
   slot {class=OverlaySlot}
  }
  VerticalBox_0 {
   widgetClassPath=&VerticalBox
   parent=Background
   widget=VerticalBox_0
   slot {class=BorderSlot, properties {padding=<ltrb 0 0 0 0>}}
  }
  HorizontalBox_0 {widgetClassPath=&HorizontalBox, parent=VerticalBox_0, widget=HorizontalBox_0, slot {class=VerticalBoxSlot}}
  TeamColor2SB {
   widgetClassPath=&SizeBox
   bIsVariable=true
   parent=HorizontalBox_0
   widget=TeamColor2SB
   properties {bOverride_WidthOverride=true, widthOverride=6}
   slot {class=HorizontalBoxSlot}
  }
  TeamColor2 {
   widgetClassPath=&CommonBorder
   bIsVariable=true
   parent=TeamColor2SB
   widget=TeamColor2
   properties {
    background {
     bIsDynamicallyLoaded=false
     drawAs=Box
     imageType=NoImage
     mirroring=NoMirror
     resourceName=None
     resourceObject=null
     tiling=NoTile
     imageSize=<32 32>
     margin=<ltrb 0 0 0 0>
     outlineSettings {
      bUseBrushTransparency=false
      roundingType=HalfHeightRadius
      width=0
      color {colorUseRule=UseColor_Specified, specifiedColor=<rgba 0 0 0 0>}
      cornerRadii=<0 0 0 0>
     }
     tintColor {colorUseRule=UseColor_Specified, specifiedColor=<rgba 1 1 1 1>}
     uVRegion {bIsValid=false, max=<0 0>, min=<0 0>}
    }
    brushColor=<rgba 1 0.2502 0 1>
    padding=<ltrb 0 0 0 0>
   }
   slot {class=SizeBoxSlot}
  }
  TeamSB {
   widgetClassPath=&SizeBox
   bIsVariable=true
   parent=HorizontalBox_0
   widget=TeamSB
   properties {bOverride_MinDesiredWidth=true, minDesiredWidth=43, visibility=Collapsed, widthOverride=42}
   slot {class=HorizontalBoxSlot}
  }
  TeamColor {
   widgetClassPath=&CommonBorder
   bIsVariable=true
   parent=TeamSB
   widget=TeamColor
   properties {
    horizontalAlignment=HAlign_Center
    verticalAlignment=VAlign_Center
    background {
     bIsDynamicallyLoaded=false
     drawAs=Image
     imageType=NoImage
     mirroring=NoMirror
     resourceName=None
     resourceObject=null
     tiling=NoTile
     imageSize=<44 44>
     margin=<ltrb 0 0 0 0>
     outlineSettings {
      bUseBrushTransparency=false
      roundingType=HalfHeightRadius
      width=0
      color {colorUseRule=UseColor_Specified, specifiedColor=<rgba 0 0 0 0>}
      cornerRadii=<0 0 0 0>
     }
     tintColor {colorUseRule=UseColor_Specified, specifiedColor=<rgba 1 1 1 1>}
     uVRegion {bIsValid=false, max=<0 0>, min=<0 0>}
    }
    brushColor=<rgba 1 0.25 0 1>
    padding=<ltrb 3 3 3 1>
   }
   slot {class=SizeBoxSlot}
  }
  TeamNumber {
   widgetClassPath=&CommonTextBlock
   bIsVariable=true
   parent=TeamColor
   widget=TeamNumber
   properties {
    justification=Center
    text=100
    colorAndOpacity {colorUseRule=UseColor_Specified, specifiedColor=<rgba 0 0 0 0.4>}
    font {
     bForceMonospaced=false
     bMaterialIsStencil=false
     fontMaterial=null
     fontObject=&Burbank
     letterSpacing=0
     monospacedWidth=1
     size=18
     skewAmount=0
     typefaceFontName=BurbankRegular-Black
     outlineSettings {
      bApplyOutlineToDropShadows=false
      bMiteredCorners=false
      bSeparateFillAlpha=false
      outlineMaterial=null
      outlineSize=0
      outlineColor=<rgba 0 0 0 1>
     }
    }
    shadowOffset=<0 0>
   }
   slot {class=BorderSlot, properties {horizontalAlignment=HAlign_Center, verticalAlignment=VAlign_Center, padding=<ltrb 3 3 3 1>}}
  }
  HitPointBarsSB {
   widgetClassPath=&SizeBox
   parent=HorizontalBox_0
   widget=HitPointBarsSB
   properties {bOverride_WidthOverride=true, widthOverride=140}
   slot {class=HorizontalBoxSlot}
  }
  CommonBorder_0 {
   widgetClassPath=&CommonBorder
   parent=HitPointBarsSB
   widget=CommonBorder_0
   properties {
    verticalAlignment=VAlign_Center
    background {
     bIsDynamicallyLoaded=false
     drawAs=Image
     imageType=NoImage
     mirroring=NoMirror
     resourceName=None
     resourceObject=null
     tiling=NoTile
     imageSize=<32 32>
     margin=<ltrb 0 0 0 0>
     outlineSettings {
      bUseBrushTransparency=false
      roundingType=HalfHeightRadius
      width=0
      color {colorUseRule=UseColor_Specified, specifiedColor=<rgba 0 0 0 0>}
      cornerRadii=<0 0 0 0>
     }
     tintColor {colorUseRule=UseColor_Specified, specifiedColor=<rgba 1 1 1 1>}
     uVRegion {bIsValid=false, max=<0 0>, min=<0 0>}
    }
    brushColor=<rgba 0 0 0 0.4>
    padding=<ltrb 4 4 4 4>
   }
   slot {class=SizeBoxSlot}
  }
  VerticalBox_1 {
   widgetClassPath=&VerticalBox
   parent=CommonBorder_0
   widget=VerticalBox_1
   slot {class=BorderSlot, properties {verticalAlignment=VAlign_Center, padding=<ltrb 4 4 4 4>}}
  }
  ScaleBox_0 {
   widgetClassPath=&ScaleBox
   parent=VerticalBox_1
   widget=ScaleBox_0
   slot {class=VerticalBoxSlot, properties {horizontalAlignment=HAlign_Center, padding=<ltrb 0 0 0 2>}}
  }
  PlayerName {
   widgetClassPath=&CommonTextBlock
   bIsVariable=true
   parent=ScaleBox_0
   widget=PlayerName
   properties {
    justification=Center
    text="{Player Name}"
    font {
     bForceMonospaced=false
     bMaterialIsStencil=false
     fontMaterial=null
     fontObject=&NotoSans
     letterSpacing=0
     monospacedWidth=1
     size=14
     skewAmount=0
     typefaceFontName=NotoSans-Bold
     outlineSettings {
      bApplyOutlineToDropShadows=false
      bMiteredCorners=false
      bSeparateFillAlpha=false
      outlineMaterial=null
      outlineSize=0
      outlineColor=<rgba 0 0 0 1>
     }
    }
    shadowOffset=<0 0>
   }
   slot {class=ScaleBoxSlot}
  }
  VehicleHealthSB {
   widgetClassPath=&SizeBox
   bIsVariable=true
   parent=VerticalBox_1
   widget=VehicleHealthSB
   properties {bOverride_HeightOverride=true, heightOverride=8}
   slot {class=VerticalBoxSlot, properties {padding=<ltrb 0 0 0 2>}}
  }
  VehicleHealthBar {
   widgetClassPath=/Game/Athena/HUD/Spectator/AthenaSpectatorHitPointBar
   bIsVariable=true
   parent=VehicleHealthSB
   widget=VehicleHealthBar
   properties {barType=VehicleHealth}
   slot {class=SizeBoxSlot}
  }
  ShieldBarSB {
   widgetClassPath=&SizeBox
   bIsVariable=true
   parent=VerticalBox_1
   widget=ShieldBarSB
   properties {bOverride_HeightOverride=true, heightOverride=8}
   slot {class=VerticalBoxSlot, properties {padding=<ltrb 0 0 0 2>}}
  }
  ShieldBar {
   widgetClassPath=&AthenaSpectatorHitPointBar
   bIsVariable=true
   parent=ShieldBarSB
   widget=ShieldBar
   properties {barType=Shield, visibility=Visible}
   slot {class=SizeBoxSlot}
  }
  HealthBarSB {
   widgetClassPath=&SizeBox
   bIsVariable=true
   parent=VerticalBox_1
   widget=HealthBarSB
   properties {bOverride_HeightOverride=true, heightOverride=8}
   slot {class=VerticalBoxSlot}
  }
  HealthBar {
   widgetClassPath=&AthenaSpectatorHitPointBar
   bIsVariable=true
   parent=HealthBarSB
   widget=HealthBar
   slot {class=SizeBoxSlot}
  }
  SessionRankBorder {
   widgetClassPath=&CommonBorder
   bIsVariable=true
   parent=HorizontalBox_0
   widget=SessionRankBorder
   properties {
    verticalAlignment=VAlign_Center
    background {
     bIsDynamicallyLoaded=false
     drawAs=Image
     imageType=NoImage
     mirroring=NoMirror
     resourceName=None
     resourceObject=null
     tiling=NoTile
     imageSize=<32 32>
     margin=<ltrb 0 0 0 0>
     outlineSettings {
      bUseBrushTransparency=false
      roundingType=HalfHeightRadius
      width=0
      color {colorUseRule=UseColor_Specified, specifiedColor=<rgba 0 0 0 0>}
      cornerRadii=<0 0 0 0>
     }
     tintColor {colorUseRule=UseColor_Specified, specifiedColor=<rgba 1 1 1 1>}
     uVRegion {bIsValid=false, max=<0 0>, min=<0 0>}
    }
    brushColor=<rgba 0 0 0 0.4>
    padding=<ltrb 0 0 0 0>
   }
   slot {class=HorizontalBoxSlot}
  }
  SessionRankBox {
   widgetClassPath=&HorizontalBox
   parent=SessionRankBorder
   widget=SessionRankBox
   slot {class=BorderSlot, properties {verticalAlignment=VAlign_Center, padding=<ltrb 0 0 0 0>}}
  }
  SessionRankIndicator {
   widgetClassPath=&CommonTextBlock
   parent=SessionRankBox
   widget=SessionRankIndicator
   properties {
    style=/FortUILibrary/Styles/Typography/Foundation/TextStyle-BurbankSmall-XS-White.TextStyle-BurbankSmall-XS-White_C
    text="##"
    font {
     bForceMonospaced=false
     bMaterialIsStencil=false
     fontMaterial=null
     fontObject=&BurbankSmall
     letterSpacing=0
     monospacedWidth=1
     size=13
     skewAmount=0
     typefaceFontName=Black
     outlineSettings {
      bApplyOutlineToDropShadows=false
      bMiteredCorners=false
      bSeparateFillAlpha=false
      outlineMaterial=null
      outlineSize=0
      outlineColor=<rgba 0 0 0 1>
     }
    }
    shadowOffset=<0 0>
   }
   slot {class=HorizontalBoxSlot, properties {verticalAlignment=VAlign_Bottom}}
  }
  SessionRank {
   widgetClassPath=&CommonTextBlock
   bIsVariable=true
   parent=SessionRankBox
   widget=SessionRank
   properties {
    style=/FortUILibrary/Styles/Typography/Foundation/TextStyle-BurbankSmall-S-White.TextStyle-BurbankSmall-S-White_C
    text=99
    font {
     bForceMonospaced=false
     bMaterialIsStencil=false
     fontMaterial=null
     fontObject=&BurbankSmall
     letterSpacing=0
     monospacedWidth=1
     size=17
     skewAmount=0
     typefaceFontName=Black
     outlineSettings {
      bApplyOutlineToDropShadows=false
      bMiteredCorners=false
      bSeparateFillAlpha=false
      outlineMaterial=null
      outlineSize=0
      outlineColor=<rgba 0 0 0 1>
     }
    }
    shadowOffset=<0 0>
   }
   slot {class=HorizontalBoxSlot, properties {padding=<ltrb 0 0 10 0>}}
  }
  TutorialHighlight {widgetClassPath=&TutorialHighlight}
 ]
}
