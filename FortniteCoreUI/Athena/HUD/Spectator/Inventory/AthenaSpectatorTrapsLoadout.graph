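# Widget Blueprint for the Athena spectator HUD's traps loadout, showing six inventory slots in a two-row, three-column grid.
# A vertical box stacks two horizontal rows of AthenaSpectatorInventorySlot widgets, separated by small spacers.
# PreConstruct gathers the six slots in order into the ItemSlots array of its parent class, AthenaSpectatorTrapsLoadoutBase.
# The Blueprint adds no variables or other graph logic of its own.
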WidgetBlueprint /FortniteCoreUI/Athena/HUD/Spectator/Inventory/AthenaSpectatorTrapsLoadout
graphs [EventGraph {nodes [
   Event PreConstruct()
   > Set ItemSlots(ItemSlots:Array of Athena Spectator Inventory Slot Base Object References=Make Array!(
     "[0]":Athena Spectator Inventory Slot Base Object Reference=Get Slot1() -> (Slot1:Athena Spectator Inventory Slot Object Reference)
     "[1]":Athena Spectator Inventory Slot Base Object Reference=Get Slot2() -> (Slot2:Athena Spectator Inventory Slot Object Reference)
     "[2]":Athena Spectator Inventory Slot Base Object Reference=Get Slot3() -> (Slot3:Athena Spectator Inventory Slot Object Reference)
     "[3]":Athena Spectator Inventory Slot Base Object Reference=Get Slot4() -> (Slot4:Athena Spectator Inventory Slot Object Reference)
     "[4]":Athena Spectator Inventory Slot Base Object Reference=Get Slot5() -> (Slot5:Athena Spectator Inventory Slot Object Reference)
     "[5]":Athena Spectator Inventory Slot Base Object Reference=Get Slot6() -> (Slot6:Athena Spectator Inventory Slot Object Reference)
    ) -> (Array:Array of Athena Spectator Inventory Slot Base Object References)) -> (Output_Get:Array of Athena Spectator Inventory Slot Base Object References)
  ]}]
variables []
widgets {
 info {
  inheritedWidgetCount=1
  parentClass=/Script/FortniteUI.AthenaSpectatorTrapsLoadoutBase
  rootWidgetClass=&VerticalBox
  widgetCount=15
 }
 widgets [
  VerticalBox_0 {widgetClassPath=&VerticalBox, widget=VerticalBox_0}
  HorizontalBox_0 {widgetClassPath=&HorizontalBox, parent=VerticalBox_0, widget=HorizontalBox_0, slot {class=VerticalBoxSlot}}
  Slot1 {
   widgetClassPath=/Game/Athena/HUD/Spectator/Inventory/AthenaSpectatorInventorySlot
   bIsVariable=true
   parent=HorizontalBox_0
   widget=Slot1
   slot {class=HorizontalBoxSlot}
  }
  Spacer_0 {widgetClassPath=&Spacer, parent=HorizontalBox_0, widget=Spacer_0, properties {size=<4 0>}, slot {class=HorizontalBoxSlot}}
  Slot2 {
   widgetClassPath=&AthenaSpectatorInventorySlot
   bIsVariable=true
   parent=HorizontalBox_0
   widget=Slot2
   slot {class=HorizontalBoxSlot}
  }
  Spacer_1 {widgetClassPath=&Spacer, parent=HorizontalBox_0, widget=Spacer_1, properties {size=<4 0>}, slot {class=HorizontalBoxSlot}}
  Slot3 {
   widgetClassPath=&AthenaSpectatorInventorySlot
   bIsVariable=true
   parent=HorizontalBox_0
   widget=Slot3
   slot {class=HorizontalBoxSlot}
  }
  Spacer_2 {widgetClassPath=&Spacer, parent=VerticalBox_0, widget=Spacer_2, properties {size=<0 4>}, slot {class=VerticalBoxSlot}}
  HorizontalBox_2 {widgetClassPath=&HorizontalBox, parent=VerticalBox_0, widget=HorizontalBox_2, slot {class=VerticalBoxSlot}}
  Slot4 {
   widgetClassPath=&AthenaSpectatorInventorySlot
   bIsVariable=true
   parent=HorizontalBox_2
   widget=Slot4
   slot {class=HorizontalBoxSlot}
  }
  Spacer_3 {widgetClassPath=&Spacer, parent=HorizontalBox_2, widget=Spacer_3, properties {size=<4 0>}, slot {class=HorizontalBoxSlot}}
  Slot5 {
   widgetClassPath=&AthenaSpectatorInventorySlot
   bIsVariable=true
   parent=HorizontalBox_2
   widget=Slot5
   slot {class=HorizontalBoxSlot}
  }
  Spacer_4 {widgetClassPath=&Spacer, parent=HorizontalBox_2, widget=Spacer_4, properties {size=<4 0>}, slot {class=HorizontalBoxSlot}}
  Slot6 {
   widgetClassPath=&AthenaSpectatorInventorySlot
   bIsVariable=true
   parent=HorizontalBox_2
   widget=Slot6
   slot {class=HorizontalBoxSlot}
  }
  TutorialHighlight {widgetClassPath=&TutorialHighlight}
 ]
}
