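# Spectator HUD widget that lays out a weapon loadout quickbar as a horizontal row of five inventory slot widgets separated by spacers.
# It derives from AthenaSpectatorQuickbarBase and only fills that class's ItemSlots array on PreConstruct, with no other logic in the Blueprint.
# The array leaves index 0 empty, so Slot1 through Slot5 sit at their matching indices.
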WidgetBlueprint /FortniteCoreUI/Athena/HUD/Spectator/Inventory/AthenaSpectatorWeaponLoadout
graphs [EventGraph {nodes [
   Event PreConstruct()
   > Set ItemSlots(ItemSlots:Array of Athena Spectator Inventory Slot Base Object References=Make Array!(
     "[0]":Athena Spectator Inventory Slot Base Object Reference
     "[1]":Athena Spectator Inventory Slot Base Object Reference=Get Slot1() -> (Slot1:Athena Spectator Inventory Slot Object Reference)
     "[2]":Athena Spectator Inventory Slot Base Object Reference=Get Slot2() -> (Slot2:Athena Spectator Inventory Slot Object Reference)
     "[3]":Athena Spectator Inventory Slot Base Object Reference=Get Slot3() -> (Slot3:Athena Spectator Inventory Slot Object Reference)
     "[4]":Athena Spectator Inventory Slot Base Object Reference=Get Slot4() -> (Slot4:Athena Spectator Inventory Slot Object Reference)
     "[5]":Athena Spectator Inventory Slot Base Object Reference=Get Slot5() -> (Slot5:Athena Spectator Inventory Slot Object Reference)
    ) -> (Array:Array of Athena Spectator Inventory Slot Base Object References)) -> (Output_Get:Array of Athena Spectator Inventory Slot Base Object References)
  ]}]
variables []
widgets {
 info {
  inheritedWidgetCount=1
  parentClass=/Script/FortniteUI.AthenaSpectatorQuickbarBase
  rootWidgetClass=&HorizontalBox
  widgetCount=12
 }
 widgets [
  HorizontalBox_0 {widgetClassPath=&HorizontalBox, widget=HorizontalBox_0}
  Slot1 {
   widgetClassPath=/Game/Athena/HUD/Spectator/Inventory/AthenaSpectatorInventorySlot
   bIsVariable=true
   parent=HorizontalBox_0
   widget=Slot1
   slot {class=HorizontalBoxSlot}
  }
  Spacer_0 {widgetClassPath=&Spacer, parent=HorizontalBox_0, widget=Spacer_0, properties {size=<8 0>}, slot {class=HorizontalBoxSlot}}
  Slot2 {
   widgetClassPath=&AthenaSpectatorInventorySlot
   bIsVariable=true
   parent=HorizontalBox_0
   widget=Slot2
   slot {class=HorizontalBoxSlot}
  }
  Spacer_1 {widgetClassPath=&Spacer, parent=HorizontalBox_0, widget=Spacer_1, properties {size=<8 0>}, slot {class=HorizontalBoxSlot}}
  Slot3 {
   widgetClassPath=&AthenaSpectatorInventorySlot
   bIsVariable=true
   parent=HorizontalBox_0
   widget=Slot3
   slot {class=HorizontalBoxSlot}
  }
  Spacer_2 {widgetClassPath=&Spacer, parent=HorizontalBox_0, widget=Spacer_2, properties {size=<8 0>}, slot {class=HorizontalBoxSlot}}
  Slot4 {
   widgetClassPath=&AthenaSpectatorInventorySlot
   bIsVariable=true
   parent=HorizontalBox_0
   widget=Slot4
   slot {class=HorizontalBoxSlot}
  }
  Spacer_3 {widgetClassPath=&Spacer, parent=HorizontalBox_0, widget=Spacer_3, properties {size=<8 0>}, slot {class=HorizontalBoxSlot}}
  Slot5 {
   widgetClassPath=&AthenaSpectatorInventorySlot
   bIsVariable=true
   parent=HorizontalBox_0
   widget=Slot5
   slot {class=HorizontalBoxSlot}
  }
  Spacer_4 {widgetClassPath=&Spacer, parent=HorizontalBox_0, widget=Spacer_4, properties {size=<8 0>}, slot {class=HorizontalBoxSlot}}
  TutorialHighlight {widgetClassPath=&TutorialHighlight}
 ]
}
